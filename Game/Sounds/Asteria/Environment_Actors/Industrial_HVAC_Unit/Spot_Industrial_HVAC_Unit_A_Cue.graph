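# Looping spot ambience cue for the Industrial HVAC Unit A environment actor, played in the Athena ambient sound class with limited concurrency.
# Mixes a louder, pitched-up BuildaBrella fan loop with short falloff and a quieter air duct fan loop that carries farther.
# Each layer has pitch modulation and its own attenuation with distance low-pass, occlusion and listener focus.
# The quality switch feeds this mix to the first two levels and leaves the other two empty.
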SoundCue /Game/Sounds/Asteria/Environment_Actors/Industrial_HVAC_Unit/Spot_Industrial_HVAC_Unit_A_Cue
root=QualityLevel_0
nodes [
 {
  id=Attenuation_1
  children [Modulator_3]
  properties {
   bOverrideAttenuation=true
   attenuationOverrides {
    bAttenuateWithLPF=true
    bEnableFocusInterpolation=true
    bEnableListenerFocus=true
    bEnableOcclusion=true
    bEnableReverbSend=false
    falloffDistance=700
    focusReleaseInterpSpeed=0.75
    lPFFrequencyAtMax=2000
    lPFRadiusMax=3000
    lPFRadiusMin=500
    nonFocusVolumeAttenuation=0.5
    occlusionInterpolationTime=1
    occlusionLowPassFilterFrequency=2000
    occlusionTraceChannel=ECC_GameTraceChannel11
    occlusionVolumeAttenuation=0.5
    stereoSpread=512
    attenuationShapeExtents {x=250}
   }
  }
 }
 {
  id=Attenuation_6
  children [Modulator_1]
  properties {
   bOverrideAttenuation=true
   attenuationOverrides {
    bAttenuateWithLPF=true
    bEnableFocusInterpolation=true
    bEnableListenerFocus=true
    bEnableOcclusion=true
    bEnableReverbSend=false
    falloffDistance=350
    focusReleaseInterpSpeed=0.5
    lPFFrequencyAtMax=2000
    lPFRadiusMax=3000
    lPFRadiusMin=500
    nonFocusVolumeAttenuation=0.5
    nonSpatializedRadiusStart=100
    occlusionInterpolationTime=1
    occlusionLowPassFilterFrequency=2000
    occlusionTraceChannel=ECC_GameTraceChannel11
    occlusionVolumeAttenuation=0.5
    attenuationShapeExtents {x=100}
   }
  }
 }
 {id=Mixer_0, children [Attenuation_6, Attenuation_1], properties {inputVolume [1.2,0.6]}}
 {id=Modulator_1, children [WavePlayer_3], properties {pitchMax=1.2, pitchMin=1.1}}
 {id=Modulator_3, children [WavePlayer_1], properties {pitchMin=0.9}}
 {id=QualityLevel_0, children [Mixer_0,Mixer_0,null,null]}
 {id=WavePlayer_1, properties {bLooping=true, soundWaveAssetPtr=/Game/Sounds/Apollo/Props/AirDucts/AMB_Air_Duct_Fan_Loop_01}}
 {id=WavePlayer_3, properties {bLooping=true, soundWaveAssetPtr=/Game/Sounds/Fort_UI_Sounds/S13_Frontend/S13_BuildaBrella_Fan_Loop}}
]
properties {
 bOverrideConcurrency=true
 maxDistance=950
 soundClassObject=&Fort_Ambient_Athena
 volumeMultiplier=0.501
 attenuationOverrides {attenuationShapeExtents {x=0}}
 concurrencyOverrides {maxCount=3}
}
