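# Looping ambient cue for an industrial HVAC unit spot sound, in the Fort_Ambient_Athena sound class with concurrency capped at three.
# A random start delay feeds a quality-level switch that routes its first two slots to a shared branch and leaves the other two empty.
# That branch applies attenuation with distance low-pass, occlusion and listener focus, then pitch modulation over a random pick of three equally weighted looping waves.
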SoundCue /Game/Sounds/Asteria/Environment_Actors/Industrial_HVAC_Unit/Spot_Industrial_HVAC_Unit_E_Cue
root=Delay_0
nodes [
 {
  id=Attenuation_1
  children [Modulator_0]
  properties {
   bOverrideAttenuation=true
   attenuationOverrides {
    bAttenuateWithLPF=true
    bEnableFocusInterpolation=true
    bEnableListenerFocus=true
    bEnableOcclusion=true
    bEnableReverbSend=false
    falloffDistance=700
    focusReleaseInterpSpeed=0.25
    lPFFrequencyAtMax=2000
    lPFRadiusMax=3000
    lPFRadiusMin=500
    nonFocusVolumeAttenuation=0.5
    occlusionInterpolationTime=1
    occlusionLowPassFilterFrequency=2000
    occlusionTraceChannel=ECC_GameTraceChannel11
    occlusionVolumeAttenuation=0.5
    attenuationShapeExtents {x=300}
   }
  }
 }
 {id=Delay_0, children [QualityLevel_0], properties {delayMax=2}}
 {id=Modulator_0, children [Random_0], properties {pitchMax=0.1, pitchMin=0.9}}
 {id=QualityLevel_0, children [Attenuation_1, Attenuation_1, null, null]}
 {id=Random_0, children [WavePlayer_1,WavePlayer_0,WavePlayer_2], properties {weights [1,1,1]}}
 {id=WavePlayer_0, properties {bLooping=true, soundWaveAssetPtr=/Game/Sounds/Asteria/POI/IndustrialTown/Spots/Spot_AC_Duct_Air_03}}
 {
  id=WavePlayer_1
  properties {bLooping=true, soundWaveAssetPtr=&/Spot_HVAC_Industrial_Unit_E}
 }
 {id=WavePlayer_2, properties {bLooping=true, soundWaveAssetPtr=/Game/Sounds/Asteria/POI/IndustrialTown/Spots/Spot_AC_Duct_Air_02}}
]
properties {
 bOverrideConcurrency=true
 maxDistance=1000
 soundClassObject=&Fort_Ambient_Athena
 volumeMultiplier=0.351
 attenuationOverrides {attenuationShapeExtents {x=0}}
 concurrencyOverrides {maxCount=3}
}
